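# Niagara system for the Fusion Energy vehicle boost cosmetic, drawing blue exhaust effects at each filtered socket on the vehicle's skeletal mesh.
# NE_Boost spawns quality-scaled, tapering ribbons, NE_Flare bursts persistent port flares, and NE_Sprite adds growing, fading puffs that drag slows.
# Ribbon and flare each layer three renderers using extra per-particle color and size attributes.
# GameModeID selects ribbon colors and flare offset, DTCarOffset shifts all emitters, and painted-variant overrides are disabled.
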NiagaraSystem /VehicleCosmetics/Boosts/Boost_FusionEnergy/FX/NS_FusionEnergy
dependencies {
 usedDataInterfaces [&NiagaraDataInterfaceCurve, &NiagaraDataInterfaceSkeletalMesh]
 usedDynamicInputs []
 usedModules [
  &self:SystemSpawnScript.SystemScriptSource.SystemScriptGraph.NiagaraNodeAssignment_0.SetVariables_#1
  &SystemState
 ]
}
emitters {
 NE_Boost {
  emitterUpdateScript [
   &EmitterState(Inactive Response=::NewEnumerator1, Life Cycle Mode=::NewEnumerator1, Loop Duration=0.3, UseLoopDelay=false)
   &VC_FilteredSocketCount()
   &VC_RateQualityHandler(Epic Spawn Rate=40, High Spawn Rate=40, Low Spawn Rate=30, Medium Spawn Rate=36, Use Custom Rate?=true)
   &VC_SpawnRibbon_v1(Number of Ribbons=$Emitter.PortCount, SpawnRate=$Emitter.Spawn Rate)
  ]
  particleSpawnScript [
   &VC_PaintedVariantsOverride() {enabled=false}
   &VC_InitializeRibbon_v1(
    Color=&MergeHelper_ColorSwitch(0=<rgba 0.04 0.078 0.8 0.5>, 1=<rgba 0.04 0.1667 0.8 0.5>, 2=<rgba 0.04 0.1667 0.8 0.5>, 3=?, Index=$User.GameModeID, Num=::NewEnumerator1)
    Custom Offset=true
    Enable Skeletal Position=true
    Lifetime=0.9
    Local Vector Offset=&AddVectorToPosition(Position=<0 0 0>, Vector=$User.DTCarOffset)
    Offset Amount=1
    Ribbon Width=70
   )
   &AddVelocity(Velocity=<-1800 0 0>) {module=AddVelocity001}
  ]
  particleUpdateScript [
   &ParticleState()
   &InheritSourceMovement(ApplyForce=false, ApplyPosition=true, ApplyPositionFalloff=false)
   &SolveForcesAndVelocity(Clamp Velocity=false, Limit Acceleration=false)
   &VC_UpdateRibbon_v1(Enable Skeletal Position=true)
   &ScaleRibbonWidth(Ribbon Width Scale=&FloatFromCurve(FloatCurve=DataInterface() {propertyValues {Curve {keys [<at 0 1 RCIM_Cubic>,<at 1 0.35 RCIM_Cubic>]}}}))
   &ScaleColor(
    Scale RGB=&VectorFromFloat(Value=&FloatFromCurve(
      FloatCurve=DataInterface() {propertyValues {Curve {
         defaultValue=1
         keys [
          {arriveTangent=-3.8255, interpMode=RCIM_Cubic, leaveTangent=-3.8255, tangentMode=RCTM_User, time=0, value=2}
          {interpMode=RCIM_Cubic, tangentMode=RCTM_User, time=0.65, value=0.4}
         ]
        }}}
      Scale Curve=6
     ))
    ScaleA=true
    ScaleRGB=true
   )
   /SetVariables(
    Particles.colorB:LinearColor=&MultiplyLinearColor(
     Linear Color=&MergeHelper_ColorSwitch(0=<rgba 0.04 0.0653 0.8 0.5>, 1=<rgba 0.04 0.1413 0.8 0.5>, 2=<rgba 0.04 0.1413 0.8 0.5>, 3=?, Index=$User.GameModeID, Num=::NewEnumerator1)
     Scale Factor=&MakeVector4FromVectorAndFloat(
      "Vector (RGB)"=&VectorFromFloat(Value=&FloatFromCurve(
        FloatCurve=DataInterface() {propertyValues {Curve {
           defaultValue=1
           keys [
            {arriveTangent=-3.8255, interpMode=RCIM_Cubic, leaveTangent=-3.8255, tangentMode=RCTM_User, time=0, value=2}
            {interpMode=RCIM_Cubic, tangentMode=RCTM_User, time=0.65, value=0.4}
           ]
          }}}
        Scale Curve=1.5
       ))
      "Float (Alpha)"=1
     )
    )
    Particles.colorC:LinearColor=&MultiplyLinearColor(
     Linear Color=&MergeHelper_ColorSwitch(
      0=<rgba 0.085 0.1487 0.85 0.5>
      1=<rgba 0.085 0.2762 0.85 0.5>
      2=<rgba 0.085 0.2762 0.85 0.5>
      3=?
      Index=$User.GameModeID
      Num=::NewEnumerator1
     )
     Scale Factor=&MakeVector4FromVectorAndFloat(
      "Vector (RGB)"=&VectorFromFloat(Value=&FloatFromCurve(
        FloatCurve=DataInterface() {propertyValues {Curve {
           defaultValue=1
           keys [
            {arriveTangent=-3.8255, interpMode=RCIM_Cubic, leaveTangent=-3.8255, tangentMode=RCTM_User, time=0, value=2}
            {interpMode=RCIM_Cubic, tangentMode=RCTM_User, time=0.65, value=0.4}
           ]
          }}}
        Scale Curve=4.2
       ))
      "Float (Alpha)"=1
     )
    )
    Particles.sizeB:NiagaraFloat=&Multiply_Float(
     A=73
     B=&FloatFromCurve(FloatCurve=DataInterface() {propertyValues {Curve {keys [<at 0 1 RCIM_Cubic>,<at 1 0.35 RCIM_Cubic>]}}})
    )
    Particles.sizeC:NiagaraFloat=&Multiply_Float(
     A=60
     B=&FloatFromCurve(FloatCurve=DataInterface() {propertyValues {Curve {keys [<at 0 1 RCIM_Cubic>,<at 1 0.35 RCIM_Cubic>]}}})
    )
   ) {isSetParameters=true}
  ]
  rendererClasses [&NiagaraRibbonRendererProperties]
  renderers [
   {rendererClass=&NiagaraRibbonRendererProperties}
   {rendererClass=&NiagaraRibbonRendererProperties, rendererIndex=1}
   {rendererClass=&NiagaraRibbonRendererProperties, rendererIndex=2}
  ]
 }
 NE_Flare {
  emitterUpdateScript [
   &EmitterState(
    Inactive Response=::NewEnumerator1
    Life Cycle Mode=::NewEnumerator1
    Loop Behavior=::NewEnumerator1
    Loop Duration Mode=::NewEnumerator1
    UseLoopDelay=false
   )
   &VC_FilteredSocketCount()
   &SpawnRate(SpawnRate=50, Use Spawn Probability=false) {enabled=false}
   &SpawnBurst_Instantaneous(Spawn Count=4, Use Spawn Probability=false, UseLoopCountLimit=false)
  ]
  particleSpawnScript [
   &VC_PaintedVariantsOverride() {enabled=false}
   &InitializeParticle(
    Color=<rgba 0.0191 0.1388 0.9167 0.4>
    Sprite Size=<510 22>
    Sprite Size Mode=::NewEnumerator4
    UsePositionOffset=false
    Write Material Random=true
   )
   &PartitionParticles(Number Of Partitions=$Emitter.PortCount)
   &SkeletalMeshLocation(
    Bone / Socket Index=$Particles.PartitionParticles.Partition
    Bone / Socket Sampling Mode=::NewEnumerator5
    MaskSpawnGroup=false
    Mesh Sampling Type=::NewEnumerator6
    Orientation Sampling=::NewEnumerator1
    Sampled Position Offset=&AddVectorToPosition(
     Position=&MergeHelper_VectorSwitch(0=<-4 0 0>, 1=<-5 0 0>, 2=<-5 0 0>, 3=?, Index=$User.GameModeID, Num=::NewEnumerator1)
     Vector=$User.DTCarOffset
    )
    Skeletal Mesh=$System.SkeletalMesh
   )
   &AddVelocity(Velocity=<-1500 0 0>) {enabled=false}
   /SetVariables(
    Particles.colorB:LinearColor=&MultiplyLinearColor(
     Linear Color=<rgba 0.194 0.5044 0.97 0.4>
     Scale Factor=&MakeVector4FromVectorAndFloat("Vector (RGB)"=&VectorFromFloat(Value=1.1), "Float (Alpha)"=1)
    )
    Particles.colorC:LinearColor=&MultiplyLinearColor(
     Linear Color=<rgba 0.4305 0.7768 0.95 0.25>
     Scale Factor=&MakeVector4FromVectorAndFloat("Vector (RGB)"=&VectorFromFloat(Value=2), "Float (Alpha)"=1)
    )
    Particles.offset=30
    Particles.sizeB:Vector2f=<290 18>
    Particles.sizeC:Vector2f=<110 70>
   ) {isSetParameters=true}
   &CameraOffset(Camera Offset Amount=20)
  ]
  particleUpdateScript [
   &ParticleState(Kill Particles When Lifetime Has Elapsed=false)
   &SolveForcesAndVelocity(Clamp Velocity=false, Limit Acceleration=false) {enabled=false}
  ]
  rendererClasses [&NiagaraSpriteRendererProperties]
  renderers [
   {rendererClass=&NiagaraSpriteRendererProperties}
   {rendererClass=&NiagaraSpriteRendererProperties, rendererIndex=1}
   {rendererClass=&NiagaraSpriteRendererProperties, rendererIndex=2}
  ]
 }
 NE_Sprite {
  emitterUpdateScript [
   &EmitterState(Inactive Response=::NewEnumerator1, Life Cycle Mode=::NewEnumerator1, Loop Duration=1, UseLoopDelay=false)
   &VC_FilteredSocketCount()
   &SpawnRate(SpawnRate=44, Use Spawn Probability=false)
   /SetVariables(Emitter.smokedistort:NiagaraFloat=&FloatFromCurve(
     FloatCurve=DataInterface() {propertyValues {Curve {keys [{arriveTangent=0.4635, time=0, value=-0.0569}, <at 0.4 0>]}}}
     CurveIndex=$Emitter.Age
    )) {isSetParameters=true}
  ]
  particleSpawnScript [
   &VC_PaintedVariantsOverride() {enabled=false}
   &InitializeParticle(
    Color=<rgba 0.068 0.1983 0.85 1>
    Lifetime Max=1.2
    Lifetime Min=0.9
    Lifetime Mode=::NewEnumerator1
    Sprite Rotation Mode=::NewEnumerator2
    Sprite Size Mode=::NewEnumerator1
    Sprite UV Mode=::NewEnumerator3
    Uniform Sprite Size Max=140
    Uniform Sprite Size Min=120
    UsePositionOffset=false
    Write Material Random=true
   )
   &PartitionParticles(Number Of Partitions=$Emitter.PortCount)
   &SkeletalMeshLocation(
    Bone / Socket Index=$Particles.PartitionParticles.Partition
    Bone / Socket Sampling Mode=::NewEnumerator5
    MaskSpawnGroup=false
    Mesh Sampling Type=::NewEnumerator6
    Sampled Position Offset=&AddVectorToPosition(Position=<-50 0 0>, Vector=$User.DTCarOffset)
    Skeletal Mesh=$System.SkeletalMesh
   )
   &AddVelocity(Velocity=<-800 0 0>)
  ]
  particleUpdateScript [
   &ParticleState()
   &InheritSourceMovement(ApplyForce=false, ApplyPosition=true, ApplyPositionFalloff=false)
   &Drag(
    Drag=&FloatFromCurve(FloatCurve=DataInterface() {propertyValues {Curve {defaultValue=1, keys [<at 0 0>,<at 0.8 0.9>]}}})
    Use Linear Drag=true
    Use Rotational Drag=false
   )
   &SolveForcesAndVelocity(Clamp Velocity=false, Limit Acceleration=false)
   &ScaleSpriteSize(Uniform Curve Sprite Scale=DataInterface() {propertyValues {Curve {keys [
        {arriveTangent=0.1953, interpMode=RCIM_Cubic, leaveTangent=0.1953, leaveTangentWeight=0.4027, tangentMode=RCTM_User, time=0, value=0.65}
        {arriveTangent=3.5867, interpMode=RCIM_Cubic, leaveTangent=3.5867, tangentMode=RCTM_User, time=0.8, value=3}
       ]}}})
   &ScaleColor(
    Scale Alpha=&FloatFromCurve(
     FloatCurve=DataInterface() {propertyValues {Curve {keys [
         <at 0 1.5>
         {
          arriveTangentWeight=0.1177
          interpMode=RCIM_Cubic
          leaveTangent=-4.975
          leaveTangentWeight=0.2059
          tangentMode=RCTM_Break
          time=0.45
          value=0.8
         }
         {arriveTangent=-0.1933, interpMode=RCIM_Cubic, leaveTangent=-0.1933, tangentMode=RCTM_User, time=0.75, value=0}
        ]}}}
     Scale Curve=0.8
    )
    Scale RGB=&VectorFromFloat(Value=&FloatFromCurve(
      FloatCurve=DataInterface() {propertyValues {Curve {keys [
          {arriveTangent=0.2363, interpMode=RCIM_Cubic, leaveTangent=0.2363, leaveTangentWeight=0.3333, tangentMode=RCTM_User, time=0.04, value=8}
          {arriveTangent=-7.5756, arriveTangentWeight=0.3333, interpMode=RCIM_Cubic, leaveTangent=-7.5756, tangentMode=RCTM_User, time=0.45, value=1}
         ]}}}
      Scale Curve=1.9
     ))
    ScaleA=true
    ScaleRGB=true
   )
  ]
  rendererClasses [&NiagaraSpriteRendererProperties]
  renderers [{rendererClass=&NiagaraSpriteRendererProperties}]
 }
}
summary {
 systemName=NS_FusionEnergy
 emitters [
  {emitterName=NE_Boost, rendererClasses [&NiagaraRibbonRendererProperties]}
  {emitterName=NE_Flare, rendererClasses [&NiagaraSpriteRendererProperties]}
  {emitterName=NE_Sprite, rendererClasses [&NiagaraSpriteRendererProperties]}
 ]
 userVariables [
  User.bIsFrontendPreview {type=NiagaraBool}
  User.bIsPainted {type=NiagaraBool}
  User.ColorVariantID {type=NiagaraInt32}
  User.DTCarOffset {type=Vector3f, description="", defaultValue {struct=Vector3f, value=<0 0 0>}}
  User.GameModeID {type=NiagaraInt32}
  User.PaintedColor {type=LinearColor, description="", defaultValue {struct=LinearColor, value=<rgba 1 1 1 1>}}
  User.PaintedScalar {type=NiagaraFloat, defaultValue=1, description=""}
  User.PortCount {type=NiagaraInt32, defaultValue=2}
 ]
}
